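# Low-spec looping engine MetaSound preset for the Melonpan vehicle, inheriting all graph logic from MSS_EngineAudio_LowSpec_Parent_REV.
# It supplies the Melonpan mono REV asset and settings, plus throttle EQ and exhaust pop tuning; Fort.Vehicle and REV interfaces expose vehicle state.
# Long and short pop slots share Lambo Huracan short samples, while blowoff has no sounds and zero volume.
# Engine attenuation, low-spec concurrency and modulation patches shape playback.
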MetaSoundSource /VehicleAudio/EngineAudio/Melonpan/MSS_EngineAudio_Melonpan_LowSpec_REV_Preset
preset_parent=/VehicleAudio/Patches/MSS_EngineAudio_LowSpec_Parent_REV
inputs [
 Blowoff.Sound {type=WaveAsset:Array, access=Constructor}
 Blowoff.Volume {type=Float, access=Constructor}
 Character.Affiliation {type=Enum:FortAudioAffiliation}
 Character.Health {type=Float}
 Character.UniqueID {type=Int32}
 Downshift {type=Trigger}
 EngineBlowoff {type=Trigger}
 Exhaust.Pops.Blowoff.Volume {type=Float, access=Constructor}
 Exhaust.Pops.Long {type=WaveAsset:Array, access=Constructor}
 Exhaust.Pops.Pitch {type=Float, access=Constructor}
 Exhaust.Pops.Shift.Volume {type=Float, access=Constructor}
 Exhaust.Pops.Short {type=WaveAsset:Array, access=Constructor}
 Exhaust.Pops.Volume {type=Float, access=Constructor}
 Exhaust.Throttle.EQ.FilterQ {type=Float, access=Constructor}
 Exhaust.Throttle.EQ.FilterType {type=Enum:BiquadFilterType, access=Constructor}
 Exhaust.Throttle.EQ.OffFreq {type=Float, access=Constructor}
 Exhaust.Throttle.EQ.OffGain {type=Float, access=Constructor}
 Exhaust.Throttle.EQ.On Throttle Frequency {type=Float, access=Constructor}
 Exhaust.Throttle.EQ.OnGain {type=Float, access=Constructor}
 Exhaust.Throttle.InterpTime {type=Time, access=Constructor}
 Exhaust.Throttle.Volume.Off {type=Float, access=Constructor}
 Exhaust.Throttle.Volume.On {type=Float, access=Constructor}
 Fort.Vehicle.Acceleration {type=Float}
 Fort.Vehicle.BlowoffStrength {type=Float}
 Fort.Vehicle.BonusSpeed {type=Float}
 Fort.Vehicle.EnableShifting {type=Bool}
 Fort.Vehicle.Gear {type=Int32}
 Fort.Vehicle.IsGrounded {type=Bool}
 Fort.Vehicle.Load {type=Float}
 Fort.Vehicle.LowFuel {type=Bool}
 Fort.Vehicle.NormalSpeed {type=Float}
 Fort.Vehicle.OnBlowoff {type=Trigger}
 Fort.Vehicle.OnDownshift {type=Trigger}
 Fort.Vehicle.OnThrottleEngaged {type=Trigger}
 Fort.Vehicle.OnThrottleReleased {type=Trigger}
 Fort.Vehicle.OnUpshift {type=Trigger}
 Fort.Vehicle.Pitch {type=Float}
 Fort.Vehicle.RPM {type=Float}
 Fort.Vehicle.Throttle {type=Float}
 GearChangeTime {type=Time, access=Constructor}
 IsLocal {type=Bool}
 Mono {type=Rev Asset, access=Constructor}
 REV.EnableShifting {type=Bool}
 REV.Gear {type=Int32}
 REV.RPM {type=Float}
 REV.Throttle {type=Float}
 REV.Velocity {type=Float}
 REVVolume {type=Float}
 Settings {type=Rev Settings, access=Constructor}
 Speed_Normalized {type=Float}
 UE.Source.OnPlay {type=Trigger}
 UE.Source.Orientation.Azimuth {type=Float}
 UE.Source.Orientation.Elevation {type=Float}
 Upshift {type=Trigger}
]
interfaces [Fort.Character:1.0, Fort.Vehicle:1.0, REV:1.0, UE.OutputFormat.Mono:1.0, UE.Source.Orientation:1.0, UE.Source:1.1]
outputs [
 Downshift {type=Trigger}
 Fundamental Frequency {type=Float}
 Gear {type=Int32}
 RPM {type=Float}
 SpeedNormalized {type=Float}
 Throttle {type=Float}
 "UE.OutputFormat.Mono.Audio:0" {type=Audio}
 "UE.OutputFormat.Stereo.Audio:1" {type=Audio}
 Upshift {type=Trigger}
]
pages [Default {input_defaults {
   Blowoff.Volume=0
   Character.Affiliation=0
   Character.Health=0
   Character.UniqueID=0
   Downshift=null
   EngineBlowoff=null
   Exhaust.Pops.Blowoff.Volume=1
   Exhaust.Pops.Pitch=0
   Exhaust.Pops.Shift.Volume=0
   Exhaust.Pops.Volume=4
   Exhaust.Throttle.EQ.FilterQ=1
   Exhaust.Throttle.EQ.FilterType=4
   Exhaust.Throttle.EQ.OffFreq=1500
   Exhaust.Throttle.EQ.OffGain=-2
   Exhaust.Throttle.EQ.On Throttle Frequency=20000
   Exhaust.Throttle.EQ.OnGain=4
   Exhaust.Throttle.InterpTime=0.1
   Exhaust.Throttle.Volume.Off=0
   Exhaust.Throttle.Volume.On=0
   Fort.Vehicle.Acceleration=0
   Fort.Vehicle.BlowoffStrength=0
   Fort.Vehicle.BonusSpeed=0
   Fort.Vehicle.EnableShifting=null
   Fort.Vehicle.Gear=0
   Fort.Vehicle.IsGrounded="<MetaSoundBuilderResult.SUCCEEDED: 0>"
   Fort.Vehicle.Load=0
   Fort.Vehicle.LowFuel=null
   Fort.Vehicle.NormalSpeed=0
   Fort.Vehicle.OnBlowoff=null
   Fort.Vehicle.OnDownshift=null
   Fort.Vehicle.OnThrottleEngaged=null
   Fort.Vehicle.OnThrottleReleased=null
   Fort.Vehicle.OnUpshift=null
   Fort.Vehicle.Pitch=0
   Fort.Vehicle.RPM=0
   Fort.Vehicle.Throttle=0
   GearChangeTime=0.1
   IsLocal=null
   Mono=&/Melonpan
   REV.EnableShifting=null
   REV.Gear=1
   REV.RPM=0
   REV.Throttle=0
   REV.Velocity=0
   REVVolume=8
   Settings=&/REVSettings_Melonpan
   Speed_Normalized=0
   UE.Source.OnPlay=null
   UE.Source.Orientation.Azimuth=0
   UE.Source.Orientation.Elevation=0
   Upshift=null
   Blowoff.Sound []
   Exhaust.Pops.Long [
    /VehicleAudio/EngineAudio/Lambo/Pops/EngineAudio_LamboHuracan_ExhaustPop_Short_14
    /VehicleAudio/EngineAudio/Lambo/Pops/EngineAudio_LamboHuracan_ExhaustPop_Short_13
    /VehicleAudio/EngineAudio/Lambo/Pops/EngineAudio_LamboHuracan_ExhaustPop_Short_12
    /VehicleAudio/EngineAudio/Lambo/Pops/EngineAudio_LamboHuracan_ExhaustPop_Short_11
    /VehicleAudio/EngineAudio/Lambo/Pops/EngineAudio_LamboHuracan_ExhaustPop_Short_10
    /VehicleAudio/EngineAudio/Lambo/Pops/EngineAudio_LamboHuracan_ExhaustPop_Short_09
    /VehicleAudio/EngineAudio/Lambo/Pops/EngineAudio_LamboHuracan_ExhaustPop_Short_08
    /VehicleAudio/EngineAudio/Lambo/Pops/EngineAudio_LamboHuracan_ExhaustPop_Short_07
    /VehicleAudio/EngineAudio/Lambo/Pops/EngineAudio_LamboHuracan_ExhaustPop_Short_06
    /VehicleAudio/EngineAudio/Lambo/Pops/EngineAudio_LamboHuracan_ExhaustPop_Short_05
    /VehicleAudio/EngineAudio/Lambo/Pops/EngineAudio_LamboHuracan_ExhaustPop_Short_04
    /VehicleAudio/EngineAudio/Lambo/Pops/EngineAudio_LamboHuracan_ExhaustPop_Short_03
    /VehicleAudio/EngineAudio/Lambo/Pops/EngineAudio_LamboHuracan_ExhaustPop_Short_02
    /VehicleAudio/EngineAudio/Lambo/Pops/EngineAudio_LamboHuracan_ExhaustPop_Short_01
   ]
   Exhaust.Pops.Short [
    &EngineAudio_LamboHuracan_ExhaustPop_Short_14
    &EngineAudio_LamboHuracan_ExhaustPop_Short_13
    &EngineAudio_LamboHuracan_ExhaustPop_Short_12
    &EngineAudio_LamboHuracan_ExhaustPop_Short_11
    &EngineAudio_LamboHuracan_ExhaustPop_Short_10
    &EngineAudio_LamboHuracan_ExhaustPop_Short_09
    &EngineAudio_LamboHuracan_ExhaustPop_Short_08
    &EngineAudio_LamboHuracan_ExhaustPop_Short_07
    &EngineAudio_LamboHuracan_ExhaustPop_Short_06
    &EngineAudio_LamboHuracan_ExhaustPop_Short_05
    &EngineAudio_LamboHuracan_ExhaustPop_Short_04
    &EngineAudio_LamboHuracan_ExhaustPop_Short_03
    &EngineAudio_LamboHuracan_ExhaustPop_Short_02
    &EngineAudio_LamboHuracan_ExhaustPop_Short_01
   ]
  }}]
properties {
 assetImportData=&self:AssetImportData
 attenuationSettings=&Vehicle_Engine_Att
 audioPropertiesSheet=&APS_Cosmetic_Vehicle_Engine
 bLooping=true
 duration=10000
 maxDistance=12000
 priority=80
 qualitySetting=Vehicles_REV_Mono
 sampleRate=48000
 soundClassObject=&Fort_Vehicles_Engine
 soundSubmixObject=&Fortnite_SFX
 volume=0.8414
 concurrencySet [&Conc_Veh_EngineAudio_LowSpec, &Conc_Veh_EngineAudio_LowSpec_Actor]
 modulationSettings {
  highpassRouting=Union
  lowpassRouting=Union
  pitchRouting=Union
  volumeRouting=Union
  highpassModulationDestination {bEnableModulation=true, modulators [&PP_FortVehicles_Engine_HPF]}
  lowpassModulationDestination {bEnableModulation=true, modulators [&Patch_FortVehicles_Engine_LPF]}
  pitchModulationDestination {bEnableModulation=true, modulators [&DM_Bus_Vehicle_Engine_Pitch]}
  volumeModulationDestination {
   bEnableModulation=true
   modulators [&DM_VehicleEnginePatch, &Patch_FortVehicles_Engine_Vol]
  }
 }
}
